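# Master material for the Sweet Teriyaki Red medium male player body, a Default Lit material-attributes shader for skeletal meshes with morph target and mesh deformer support.
# It chains shared character functions: a PreFX base feeds CharacterSkinFX, then MF_Sweet, then PostFX.
# MF_Sweet adds the character-specific stage, taking the pre-skinned local position and a SkinFX_Mask texture parameter.
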Material /Game/Characters/Player/Male/Medium/Bodies/M_MED_Sweet_Teriyaki_Red/Materials/M_MED_Sweet_Master
graph {
 properties {
  bUseMaterialAttributes=true
  bUsedWithMeshDeformer=true
  bUsedWithMorphTargets=true
  bUsedWithSkeletalMesh=true
  pixelDepthOffsetMode=PDOM_Legacy
  refractionMethod=RM_IndexOfRefraction
  usedShadingModels=Default Lit
 }
 nodes [
  n1 = MaterialFunctionCall() {materialFunction=&BaseFN_CharacterFunction_PreFX}
  MaterialFunctionCall(
   AO=@n1.AO
   Distance Blend=@n1.AO
   Fresnel=@n1.AO
   Gmap=@n1.AO
   Pre Skinned Local Position=@n1.AO
   Material=MaterialFunctionCall!(
    In=MaterialFunctionCall(
     PreskinnedPosition=@n1.Pre skinned Local Position
     Gmap=@n1.Pre skinned Local Position
     MaterialAttributes=@n1.Pre skinned Local Position
     AO=@n1.Pre skinned Local Position
    ) {materialFunction=&CharacterSkinFX}.Result
    FX Mask=TextureObjectParameter() {group=Skin FX, parameterName=SkinFX_Mask}
    Pre Skinned Position=@n1.Pre skinned Local Position
   ) {materialFunction=&/MF_Sweet}.Result
  ) {materialFunction=&BaseFN_CharacterFunction_PostFX}
 ]
}
